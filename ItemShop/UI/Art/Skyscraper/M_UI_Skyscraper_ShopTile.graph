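# UI-domain material that builds a Skyscraper-themed Item Shop tile image, with a grey grunge backdrop and character art.
# A rotated, smoothstepped linear gradient blends two dark greys for the base.
# The red channel of a screen-space-scaled grunge normal map, smoothstepped, pulls that base toward the lighter grey.
# A hard-coded FalseVulture outfit texture, center-scaled and offset, blends on top using its own RGB as alpha.
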Material /ItemShop/UI/Art/Skyscraper/M_UI_Skyscraper_ShopTile
graph {
 properties {bUsedWithStaticMesh=false, materialDomain=MD_UI, pixelDepthOffsetMode=PDOM_Legacy, usedShadingModels=Default Lit}
 nodes [
  n1 = Constant3Vector() {constant=<rgba 0.0401 0.0401 0.0434 1>}
  n2 = Constant3Vector() {constant=<rgba 0.1981 0.1981 0.2159 1>}
  n3 = Constant2Vector() {g=1080, r=1920}
  n4 = SmoothStep(Value=MaterialFunctionCall(
    UVs=MaterialFunctionCall(TextureSize=@n3) {materialFunction=&MF_UI_UVScaling}.Scaled To Fill
    GradientStart=Constant()
    GradientEnd=Constant() {r=0.75}
    RotationAngle=Constant() {r=0.3}
   ) {materialFunction=&MF_UI_LinearGradient}.Result) {constMax=1.75}
  n5 = TextureSample(UVs=Multiply(A=MaterialFunctionCall!(ScreenUV, TextureSize, ScalePivotPoint) {materialFunction=/FortUILibrary/MaterialFunctions/Basics/MF_UI_ScreenSpaceUVScaling}.Scaled to Fit) {constB=2}) {samplerType=SAMPLERTYPE_Normal, texture=/OfferCatalog/Art/Skyscraper/T_UI_SkyscraperGrungeTile}
  n6 = MaterialFunctionCall() {materialFunction=&MF_UI_UVScaling}
  n7 = Constant() {r=1.2}
  n8 = Constant()
  n9 = Constant() {r=-0.1}
  n10 = TextureSample(UVs=Add(
    A=MaterialFunctionCall(UVs=@n6.Scaled To Fill, Texture Scale=@n7) {materialFunction=&ScaleUVsByCenter}.UVs
    B=AppendVector(A=@n8, B=@n9)
   )) {texture=/OfferCatalog/Textures/BattleRoyale/FeaturedItems/Outfit/S28/T-AthenaSoldiers-Character-FalseVulture}
  LinearInterpolate(
   A=LinearInterpolate(A=LinearInterpolate(A=@n1, B=@n2, Alpha=@n4), B=@n2, Alpha=SmoothStep(Value=@n5.R) {constMin=-1.7})
   B=@n10.RGB
   Alpha=@n10.RGB
  )
 ]
}
